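# Material function that fakes translucent skin on the SL skin of the LiftingAura Allure character, blended into incoming material attributes by a channel-selectable mask.
# It builds Subsurface color from sky-light samples along the view direction, tinted and clamped, and emissive from reflection, Fresnel and time-of-day tints.
# An optional panning, bump-offset smoke pattern adds tinted detail; its sine pulse blends identical inputs, so has no effect.
# Mobile quality tiers use a simpler non-subsurface branch.
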MaterialFunction /BRCosmetics/Characters/Player/Female/Medium/Bodies/F_MED_LiftingAura_Allure/Skins/SL/Materials/MF_MED_LiftingAura_FakeTrans_SL
graph {nodes [
  n1 = NamedRerouteUsage()
  n2 = OneMinus(None=Fresnel(ExponentIn=ScalarParameter() {defaultValue=0.3, parameterName=SubtlePatternFresnelExponent}) {baseReflectFraction=0, exponent=0.2})
  n3 = ScalarParameter() {defaultValue=2, parameterName=SubtlePatternFresnelPow}
  n4 = TextureCoordinate()
  n5 = MaterialFunctionCall(Float2=@n4) {materialFunction=&BreakOutFloat2Components}
  n6 = MaterialFunctionCall(X=@n5.R, Y=Multiply(A=@n5.G) {constB=-1}) {materialFunction=&MakeFloat2}
  n7 = TextureSampleParameter2D() {
   parameterName=SubtleSmokeMask
   samplerType=SAMPLERTYPE_Masks
   texture=/Game/Characters/Player/Female/Medium/Bodies/F_MED_LiftingAura_Allure/Textures/T_AdditionalBubbleMask
  }
  n8 = Panner(
   Coordinate=Multiply(A=LinearInterpolate(A=@n4, B=@n6.Result, Alpha=@n7.R), B=ScalarParameter() {defaultValue=5, parameterName=HeartTilling}) {constB=10}
   Speed=AppendVector(
    A=ScalarParameter() {group=FakeTransSkin, parameterName=PannerSpeedX}
    B=ScalarParameter() {defaultValue=0.1, group=FakeTransSkin, parameterName=PannerSpeedY}
   )
  ) {speedY=1}
  n9 = TextureSampleParameter2D(UVs=BumpOffset(Coordinate=@n8, Height=ScalarParameter() {defaultValue=-0.4, parameterName=Height})) {
   group=FakeTransSkin
   parameterName=SubtleSmoke
   samplerType=SAMPLERTYPE_LinearColor
   texture=/Game/Characters/Player/Female/Medium/Bodies/F_MED_LiftingAura_Allure/Textures/T_LiftingAuraSmoke
  }
  n10 = Multiply(A=Time(), B=ScalarParameter() {defaultValue=0.3, group=FakeTransSkin, parameterName=TimeScale})
  n11 = LinearInterpolate(A=@n9.R, B=@n9.R, Alpha=ConstantBiasScale(None=Sine(None=@n10)))
  n12 = Multiply(
   A=Multiply(A=Power(Base=@n2, Exp=@n3), B=Multiply(A=@n11, B=@n7.G))
   B=VectorParameter() {group=FakeTransSkin, parameterName=HeartTint, defaultValue=<rgba 1 0 0.3612 1>}.RGB
  )
  n13 = ScalarParameter() {defaultValue=20, group=FakeTransSkin, parameterName=SubtlePatternIntensityArm}
  n14 = LinearInterpolate(
   A=LinearInterpolate(A=@n13, B=Multiply(A=@n13) {constB=2}, Alpha=NamedRerouteUsage())
   B=ScalarParameter() {defaultValue=4, group=FakeTransSkin, parameterName=SubtlePatternIntensityLeg}
   Alpha=@n7.B
  )
  n15 = NamedRerouteUsage()
  n16 = StaticSwitchParameter(True=Add(A=Multiply(A=@n12, B=@n14), B=@n15), False=@n15) {defaultValue=true, group=FakeTransSkin, parameterName=UseSubtleEffect}
  n17 = NamedRerouteUsage()
  n18 = QualitySwitch(Default=@n16, Low=@n17, Epic=@n16) {desc=has GBuffers}
  n19 = SetMaterialAttributes!(
   MaterialAttributes=@n1
   ShadingModel=ShadingModel() {shadingModel=MSM_Subsurface}
   BaseColor=Constant()
   EmissiveColor=NamedRerouteUsage()
   SubsurfaceColor=NamedRerouteDeclaration(None=ShadingPathSwitch(Default=@n18, Deferred=@n18, Mobile=@n17)) {name=SSS_OUT, nodeColor=<rgba 0.7344 0 1 1>}
   Opacity=LinearInterpolate(A=ScalarParameter() {defaultValue=0.05, group=FakeTransSkin, parameterName=SSS_Opacity}, B=Constant() {r=0.8}, Alpha=NamedRerouteUsage())
   Roughness=NamedRerouteUsage()
  ) {attributeSetTypes [
    #1
    #2
    #3
    #4
    #5
    #6
   ]}
  n20 = SetMaterialAttributes(
   MaterialAttributes=NamedRerouteUsage()
   ShadingModel=ShadingModel()
   BaseColor=NamedRerouteUsage()
   EmissiveColor=NamedRerouteUsage()
   SubsurfaceColor=Constant()
   Opacity=Constant() {r=1}
   Roughness=NamedRerouteUsage()
  ) {attributeSetTypes [
    #1
    #2
    #3
    #4
    #5
    #6
   ]}
  n21 = NamedRerouteUsage()
  FunctionOutput(None=BlendMaterialAttributes(
    A=@n1
    B=MaterialFunctionCall!(
     Low _ PC Low and Med=@n19
     "High _ XB1, PS4"=@n19
     "Epic _ PC High, XSX, PS5"=@n19
     --------
     "Mobile Low _ PC Perf, Android"=@n20
     Mobile Med _ Switch=@n20
    ) {materialFunction=/Game/Global/MaterialFunctions/TechArt/MF_Fortnite_QualitySwitch_MaterialAttributes}.Result
    Alpha=Step(X=@n21) {constY=0.1}
   ))
  n22 = GetMaterialAttributes(None=FunctionInput() {bUsePreviewValueAsDefault=true, inputName=MA, inputType=FunctionInput_MaterialAttributes}) {attributeGetTypes [#2, #3, #6]}
  NamedRerouteDeclaration(None=@n22.MaterialAttributes) {name=MA_IN, nodeColor=<rgba 0.9453 0 1 1>}
  NamedRerouteDeclaration(None=@n22.BaseColor) {name=BaseColor_IN, nodeColor=<rgba 0.3828 1 0 1>}
  NamedRerouteDeclaration(None=ScalarParameter() {group=FakeTransSkin, parameterName=IsFrontEnd}) {name=FrontEndOverride, nodeColor=<rgba 0 1 0.9063 1>}
  NamedRerouteDeclaration(None=@n22.EmissiveColor) {name=Emissive_IN, nodeColor=<rgba 0.1953 0 1 1>}
  n23 = FunctionInput() {bUsePreviewValueAsDefault=true, inputName=TechArtMask, sortPriority=1, previewValue=<1 0 0 1>}
  NamedRerouteDeclaration(None=ChannelMaskParameter(None=@n23) {group=FakeTransSkin, parameterName=UseFakeTransSkin}.R) {name=TechArtMask, nodeColor=<rgba 0.7578 1 0 1>}
  n24 = PreSkinnedPosition()
  n25 = VertexInterpolator(VS=@n24)
  NamedRerouteDeclaration(None=FunctionInput(Preview=ShaderStageSwitch(PixelShader=@n25.PS, VertexShader=@n24)) {bUsePreviewValueAsDefault=true, inputName=Pre Skinned Position, sortPriority=3}) {name=PreSkinnedPos_VS-PS, nodeColor=<rgba 0 0.4844 1 1>}
  n26 = CameraVectorWS()
  n27 = Constant2Vector()
  n28 = Multiply(A=NamedRerouteUsage(), B=ScalarParameter() {defaultValue=0.1, group=FakeTransSkin, parameterName=VMod})
  n29 = SkyLightEnvMapSample(
   Direction=Add(A=Multiply(A=@n26) {constB=-1}, B=AppendVector(A=@n27, B=@n28))
   Roughness=LinearInterpolate(A=ScalarParameter() {group=FakeTransSkin, parameterName=SkySampleRoughness01}, B=Constant() {r=0.7}, Alpha=NamedRerouteUsage())
  )
  n30 = Multiply(A=@n29) {constB=1.3}
  n31 = MaterialFunctionCall(Color_Morning=@n29, Color_Day=@n29, Color_Evening=@n30, Color_Night=@n30) {materialFunction=&Color_Time_of_Day}
  n32 = LinearInterpolate(A=ScalarParameter() {defaultValue=0.9, group=FakeTransSkin, parameterName=MaxSkyLightness}, B=Constant() {r=0.5}, Alpha=NamedRerouteUsage())
  n33 = Clamp(None=@n31.Result, Max=@n32)
  n34 = Desaturation(None=@n31.Result)
  n35 = Multiply(
   A=NamedRerouteDeclaration(None=Clamp(None=@n34, Max=@n32)) {name=SkyTint, nodeColor=<rgba 1 0.9609 0 1>}
   B=NamedRerouteDeclaration(None=VectorParameter() {group=FakeTransSkin, parameterName=FakeSkinTint, defaultValue=<rgba 0.7552 0 0.4817 1>}.RGB) {name=SkinTint, nodeColor=<rgba 0.8047 0 1 1>}
  )
  n36 = ScalarParameter() {defaultValue=1, group=FakeTransSkin, parameterName=SkyColorBlend}
  NamedRerouteDeclaration(None=Multiply(A=LinearInterpolate(A=@n33, B=@n35, Alpha=@n36), B=ScalarParameter() {defaultValue=1, group=FakeTransSkin, parameterName=SSS_Intensity})) {name=SSS_ColorOUT, nodeColor=<rgba 0.0547 1 0 1>}
  NamedRerouteDeclaration(None=@n22.Roughness) {name=Roughness_IN, nodeColor=<rgba 0 0.0391 1 1>}
  n37 = Fresnel(ExponentIn=ScalarParameter() {defaultValue=2, group=FakeTransSkin, parameterName=FresnelExp})
  NamedRerouteDeclaration(None=@n37) {name=FresnelAdj, nodeColor=<rgba 0.6172 1 0 1>}
  n38 = ScalarParameter() {defaultValue=0.2, group=FakeTransSkin, parameterName=SkySampleRoughness02}
  n39 = Constant() {r=0.3}
  n40 = NamedRerouteUsage()
  n41 = Multiply(A=SkyLightEnvMapSample(Direction=ReflectionVectorWS(), Roughness=LinearInterpolate(A=@n38, B=@n39, Alpha=@n40)), B=NamedRerouteUsage())
  n42 = ScalarParameter() {defaultValue=0.2, group=FakeTransSkin, parameterName=ReflectionColorIntensity}
  n43 = Multiply(A=Multiply(A=NamedRerouteUsage(), B=@n37), B=ScalarParameter() {defaultValue=2, group=FakeTransSkin, parameterName=FresnelIntensity})
  n44 = Multiply(A=NamedRerouteUsage(), B=ScalarParameter() {defaultValue=0.6, group=FakeTransSkin, parameterName=ORGEmissiveIntensity})
  n45 = Multiply(
   A=Add(A=Multiply(A=@n41, B=@n42), B=Add(A=@n43, B=@n44))
   B=ScalarParameter() {defaultValue=0.6, group=FakeTransSkin, parameterName=EmissiveIntensity}
  )
  n46 = VectorParameter() {group=FakeTransSkin, parameterName=EM_Tint_Evening, defaultValue=<rgba 0.6716 0.6815 0.9844 1>}
  n47 = VectorParameter() {group=FakeTransSkin, parameterName=EM_Tint_Night, defaultValue=<rgba 0.4563 0.4729 0.9844 1>}
  NamedRerouteDeclaration(None=MaterialFunctionCall(Color_Morning=@n45, Color_Day=@n45, Color_Evening=Multiply(A=@n45, B=@n46.RGB), Color_Night=Multiply(A=@n45, B=@n47.RGB)) {materialFunction=&Color_Time_of_Day}.Result) {name=EM_OUT, nodeColor=<rgba 0.9967 0 1 1>}
  Reroute(None=@n8)
 ]}
